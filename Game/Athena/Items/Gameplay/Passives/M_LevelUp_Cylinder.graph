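# Additive unlit mesh-particle material for a level-up cylinder effect, with rainbow gradient and tinted glow scaled by particle color.
# Panning, warped noise distorts the rainbow lookup and breaks up opacity, shaped by axis-aligned Fresnel, end masks and a Wave-driven vertical gradient.
# Depth fade and scene world-normal Fresnel add glow where geometry intersects, using a constant on mobile and forward paths.
# Dynamic parameter WPO flares one end horizontally from the pivot.
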Material /Game/Athena/Items/Gameplay/Passives/M_LevelUp_Cylinder
graph {
 properties {bUsedWithMeshParticles=true, blendMode=BLEND_Additive, pixelDepthOffsetMode=PDOM_Legacy, shadingModel=MSM_Unlit, usedShadingModels=Unlit}
 nodes [
  MaterialFunctionCall() {materialFunction=&DebugTimeSine}
  n1 = ParticleColor()
  n2 = TextureCoordinate() {uTiling=4, vTiling=0.7}
  n3 = ComponentMask(None=TextureCoordinate()) {g=true}
  n4 = Multiply(A=Constant2Vector() {g=-0.5}, B=Power(Base=@n3))
  n5 = Constant2Vector() {g=0.8, r=0.05}
  n6 = DynamicParameter() {paramNames [Wave,WPO,Param3,Param4]}
  n7 = Add(A=Multiply(A=Time(), B=ScalarParameter() {defaultValue=-3, parameterName=PanSpeed}), B=@n6.Wave)
  n8 = TextureSample(UVs=Add(A=Add(A=@n2, B=@n4), B=Multiply(A=@n5, B=@n7))) {texture=&T_Smooth_tile}
  n9 = SceneTexture() {sceneTextureId=PPI_WorldNormal}
  n10 = Fresnel(
   ExponentIn=ScalarParameter() {defaultValue=2, parameterName=WorldNormalFresnelExp}
   Normal=MaterialFunctionCall(Normal=ComponentMask(None=@n9.Color) {b=true, g=true, r=true}, Flatness=Constant() {r=0.4}) {materialFunction=&FlattenNormal}.Result
  ) {baseReflectFraction=0.2, exponent=1}
  n11 = Constant() {r=0.3}
  n12 = ScalarParameter() {defaultValue=44, parameterName=VisibleDepthDistance}
  n13 = Multiply(
   A=ShadingPathSwitch(Default=ShadingPathSwitch(Default=@n10, Deferred=@n10, Forward=@n11, Mobile=@n10), Mobile=@n11) {desc=Migrated from FeatureLevelSwitch}
   B=OneMinus(None=DepthFade(FadeDistance=@n12) {fadeDistanceDefault=80})
  )
  n14 = Multiply(A=Add(A=@n8.G, B=@n13)) {constB=0.35}
  n15 = TextureSample(UVs=Add(A=TextureCoordinate() {vTiling=0.5}, B=Add(A=@n14, B=@n7))) {texture=/Game/Effects/Fort_Effects/Textures/Grad/T_Gradient_Rainbow_NoRed}
  n16 = MaterialFunctionCall(
   Axis=VectorParameter() {parameterName=FresnelAxis, defaultValue=<rgba 0 0 1 0>}.RGB
   Exponent=ScalarParameter() {defaultValue=1.5, parameterName=FresnelExp}
  ) {materialFunction=&AxisAlignedFresnel}
  n17 = LinearInterpolate(B=@n13, Alpha=@n16.Fresnel) {constA=1, constB=0}
  n18 = Multiply(A=@n13, B=ScalarParameter() {defaultValue=5, parameterName=DecalMult})
  n19 = Multiply(A=Add(A=@n18, B=Power(Base=@n17)), B=VectorParameter() {parameterName=MainColor, defaultValue=<rgba 2 1.1558 0.0208 0>}.RGB)
  MaterialFunctionCall(Emissive=Multiply(A=@n1.RGB, B=Add(A=@n15.RGB, B=@n19))) {materialFunction=&MF_FX_EmissiveAdjustments}
  n20 = Constant() {r=0.5}
  n21 = SphereMask(A=@n3, B=@n20, Radius=@n20, Hardness=ScalarParameter() {defaultValue=0.7, parameterName=TopBottomHardness}) {attenuationRadius=0.5, hardnessPercent=80}
  n22 = MaterialFunctionCall(Float2=TextureCoordinate() {vTiling=0.6}) {materialFunction=&BreakOutFloat2Components}
  n23 = AppendVector(A=@n22.R, B=OneMinus(None=@n22.G))
  n24 = LinearInterpolate(Alpha=@n6.Wave) {constA=-0.95, constB=0.25}
  n25 = Power(Base=TextureSample(UVs=Add(A=@n23, B=@n24)) {texture=/Game/Effects/Fort_Effects/Textures/Masks/T_Vert_Gradient_Softer}.G)
  n26 = Add(A=@n25, B=Multiply(A=@n25, B=@n18))
  n27 = Power(Base=Saturate(None=Multiply(A=@n16.Fresnel) {constB=2})) {constExponent=2.5}
  n28 = Multiply(A=@n8.G, B=Multiply(A=Multiply(A=@n21, B=@n26), B=LinearInterpolate(B=@n17, Alpha=@n27)))
  n29 = ScalarParameter() {defaultValue=18, parameterName=DepthFade}
  Multiply(
   A=@n1.A
   B=ShadingPathSwitch(Default=DepthFade(Opacity=@n28, FadeDistance=@n29) {fadeDistanceDefault=15}, Mobile=@n28) {desc=Migrated from FeatureLevelSwitch}
  )
  n30 = Constant3Vector() {constant=<rgba 1 1 0 0>}
  n31 = Subtract(
   A=WorldPosition().XYZ
   B=MaterialFunctionCall() {materialFunction=&ObjectPivotPoint}.Object Pivot Location
  )
  n32 = OneMinus(None=ComponentMask(None=TextureCoordinate()) {g=true})
  n33 = ScalarParameter() {defaultValue=8, parameterName=WPOMaskPower}
  Multiply(A=@n6.WPO, B=LinearInterpolate(B=Multiply(A=@n30, B=@n31), Alpha=Power(Base=@n32, Exp=@n33) {constExponent=10}) {constB=0})
 ]
}
